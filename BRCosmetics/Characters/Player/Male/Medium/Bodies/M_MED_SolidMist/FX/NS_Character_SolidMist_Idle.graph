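# Idle Niagara effect for the SolidMist character body, drawing dark mesh smoke that follows bones or sockets on the skeletal mesh.
# Attached_Smoke bursts eight persistent particles, while Swiping_Smoke continuously spawns short-lived ones on random groups.
# Spawn group sets bone index and renderer visibility tag, and smoothed bone position feeds dynamic material parameters alongside animated curves.
# Color tints by time of day, and user bools hide hand or foot groups through alpha.
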NiagaraSystem /BRCosmetics/Characters/Player/Male/Medium/Bodies/M_MED_SolidMist/FX/NS_Character_SolidMist_Idle
dependencies {
 usedDataInterfaces [&NiagaraDataInterfaceCurve, &NiagaraDataInterfaceArrayFloat, &NiagaraDataInterfaceSkeletalMesh]
 usedDynamicInputs []
 usedModules [
  &self:SystemSpawnScript.SystemScriptSource.SystemScriptGraph.NiagaraNodeAssignment_9.SetVariables_#1
  &SystemState
 ]
}
emitters {
 Attached_Smoke {
  emitterUpdateScript [
   &EmitterState(Life Cycle Mode=::NewEnumerator1, Loop Behavior=::NewEnumerator1, Loop Duration Mode=::NewEnumerator1, UseLoopDelay=false)
   &SpawnBurst_Instantaneous(Spawn Count=8, Use Spawn Probability=false, UseLoopCountLimit=false)
  ]
  particleSpawnScript [
   &FN_SetSpawnGroup(Number of Spawn Groups=8)
   /SetVariables(Particles.VisibilityTag=&SpawnGroup()) {isSetParameters=true, module=SetVariables_#2}
   &InitializeParticle(
    Color=<rgba 0.015 0.0104 0.0104 1>
    Lifetime=&RandomRangeFloat(Minimum=1, Maximum=1.5)
    Mesh Scale Mode=::NewEnumerator3
    UsePositionOffset=false
    Write Material Random=true
   )
   &CameraOffset(Camera Offset Amount=2)
   /SetVariables(Particles.Y Offset:NiagaraFloat=&RandomRangeFloat(Minimum=6, Maximum=10)) {isSetParameters=true, module=SetVariables_#3}
  ]
  particleUpdateScript [
   &ParticleState(Kill Particles When Lifetime Has Elapsed=false)
   &SkeletalMeshLocation(
    Bone / Socket Index=&SpawnGroup()
    Bone Sampling Mode=::NewEnumerator4
    MaskSpawnGroup=false
    Skeletal Mesh=$System.Skeletal Mesh
   ) {module=SkeletalMeshLocation001}
   &InterpolateOverTime(Rate Of Change=5, Target Vector Value:Vector3f=$Particles.SkeletalMeshLocation001.SampledPosition, Variable Type=::NewEnumerator2)
   /SetVariables(Particles.Y Offset:NiagaraFloat=&Add_Float(A=$Particles.Y Offset, B=0)) {isSetParameters=true, module=SetVariables_#4}
   &DynamicMaterialParameters(
    Index 0 Param 1=1
    Index 0 Param 2=&Lerp_Float(
     A=0.5
     B=1
     Alpha=&FloatFromCurve(
      FloatCurve=DataInterface() {propertyValues {Curve {
         defaultValue=1
         keys [
          {interpMode=RCIM_Cubic, tangentMode=RCTM_User, time=0, value=0}
          {interpMode=RCIM_Cubic, tangentMode=RCTM_User, time=0.4, value=1}
          {tangentMode=RCTM_User, time=1, value=0}
         ]
        }}}
      CurveIndex=&ModuloFloat(
       A=&Add_Float(
        A=&Multiply_Float(A=$Emitter.Age, B=0.2)
        B=&Multiply_FloatByInt(Float=0.86, Integer=$Particles.SpawnGroup)
       )
       B=1
      )
     )
    )
    Index 0 Param 3=&Lerp_Float(
     A=1.5
     B=2
     Alpha=&FloatFromCurve(
      FloatCurve=DataInterface() {propertyValues {Curve {
         defaultValue=1
         keys [
          {interpMode=RCIM_Cubic, tangentMode=RCTM_User, time=0, value=0}
          {interpMode=RCIM_Cubic, tangentMode=RCTM_User, time=0.4, value=1}
          {tangentMode=RCTM_User, time=1, value=0}
         ]
        }}}
      CurveIndex=&ModuloFloat(
       A=&Add_Float(
        A=&Multiply_Float(A=$Emitter.Age, B=&RandomRangeFloat(Minimum=0.1, Maximum=0.25))
        B=&Multiply_FloatByInt(Float=0.17, Integer=$Particles.SpawnGroup)
       )
       B=1
      )
     )
    )
    Index 0 Param 4=&SelectFloatFromArray(
     Direct Array Index=$Particles.SpawnGroup
     Float Selection Array=DataInterface() {propertyValues {FloatData [0.8,0.8,0.9,0.9,0.8,0.8,0.8,0.8]}}
    )
    Index 1 Param 1=&MakeFloatFromVector(Vector=$StackContext.InterpolateOverTime.InterpolateOverTime_V3_.Moving Average, Channel=::NewEnumerator0)
    Index 1 Param 2=&MakeFloatFromVector(Vector=$StackContext.InterpolateOverTime.InterpolateOverTime_V3_.Moving Average, Channel=::NewEnumerator1)
    Index 1 Param 3=&MakeFloatFromVector(Vector=$StackContext.InterpolateOverTime.InterpolateOverTime_V3_.Moving Average, Channel=::NewEnumerator2)
    Index 2 Param 1=$Particles.Y Offset
    Index 2 Param 3=&SelectFloatFromArray(
     Array Sampling Mode=::NewEnumerator0
     Direct Array Index=?
     Float Selection Array=DataInterface() {propertyValues {FloatData [1,0.5,0]}}
     Random Seed=0
    )
    Module.Index0Float0Write=true
    Module.Index0Float1Write=true
    Module.Index0Float2Write=true
    Module.Index0Float3Write=true
    Module.Index1Float0Write=true
    Module.Index1Float1Write=true
    Module.Index1Float2Write=true
    Module.Index1Float3Write=true
    Module.Index2Float0Write=true
    Module.Index2Float1Write=true
    Module.Index2Float2Write=true
    Module.Index2Float3Write=true
    Write Parameter Index 1=::NewEnumerator1
    Write Parameter Index 2=::NewEnumerator1
   ) {module=DynamicMaterialParameters001}
   &ScaleColor(
    Scale Alpha=&FloatFromCurve(FloatCurve=DataInterface() {propertyValues {Curve {
        defaultValue=0
        keys [{interpMode=RCIM_Cubic, tangentMode=RCTM_User, time=0, value=0}, {interpMode=RCIM_Cubic, tangentMode=RCTM_User, time=1, value=1}]
       }}})
    Scale RGB=&MakeVectorFromLinearColorRGB(Color=&FNTimeOfDayColor(Night Color=<rgba 0.4 0.4791 0.8 1>, Morning Color=<rgba 1 1 1 1>, Day Color=<rgba 1 1 1 1>, Evening Color=<rgba 1 1 1 1>))
    ScaleA=true
    ScaleRGB=true
   )
   &ScaleColor(
    Color Value To Scale=$Particles.Color
    Scale Alpha=&MakeCustomFloatFromBool(
     Bool=&Boolean_AndOperation(
      Bool1=$User.bisFootHidden
      Bool2=&Boolean_OrOperation(
       Bool1=&SetBoolByIntComparison(A=$Particles.SpawnGroup, B=4, Comparison Type=::NewEnumerator4)
       Bool2=&Boolean_OrOperation(
        Bool1=&SetBoolByIntComparison(A=$Particles.SpawnGroup, B=5, Comparison Type=::NewEnumerator4)
        Bool2=&Boolean_OrOperation(
         Bool1=&SetBoolByIntComparison(A=$Particles.SpawnGroup, B=6, Comparison Type=::NewEnumerator4)
         Bool2=&SetBoolByIntComparison(A=$Particles.SpawnGroup, B=7, Comparison Type=::NewEnumerator4)
        )
       )
      )
     )
     True Float=0
     False Float=1
    )
    ScaleA=true
    ScaleRGB=true
   ) {module=ScaleColor001}
   &ScaleColor(
    Color Value To Scale=$Particles.Color
    Scale Alpha=&MakeCustomFloatFromBool(
     Bool=&Boolean_AndOperation(
      Bool1=$User.bisHandHidden
      Bool2=&Boolean_OrOperation(
       Bool1=&SetBoolByIntComparison(A=$Particles.SpawnGroup, B=0, Comparison Type=::NewEnumerator4)
       Bool2=&SetBoolByIntComparison(A=$Particles.SpawnGroup, B=1, Comparison Type=::NewEnumerator4)
      )
     )
     True Float=0
     False Float=1
    )
    ScaleA=true
    ScaleRGB=true
   ) {module=ScaleColor002}
  ]
  rendererClasses [&NiagaraMeshRendererProperties]
  renderers [
   {rendererClass=&NiagaraMeshRendererProperties, rendererIndex=0}
   {rendererClass=&NiagaraMeshRendererProperties, rendererIndex=1}
   {rendererClass=&NiagaraMeshRendererProperties, rendererIndex=2}
   {rendererClass=&NiagaraMeshRendererProperties, rendererIndex=3}
   {rendererClass=&NiagaraMeshRendererProperties, rendererIndex=4}
   {rendererClass=&NiagaraMeshRendererProperties, rendererIndex=5}
   {rendererClass=&NiagaraMeshRendererProperties, rendererIndex=6}
   {rendererClass=&NiagaraMeshRendererProperties, rendererIndex=7}
  ]
 }
 Swiping_Smoke {
  emitterUpdateScript [
   &EmitterState(Life Cycle Mode=::NewEnumerator1, Loop Behavior=::NewEnumerator1, Loop Duration Mode=::NewEnumerator1, UseLoopDelay=false)
   &SpawnRate(SpawnRate=25, Use Spawn Probability=false)
  ]
  particleSpawnScript [
   /SetVariables(Particles.SpawnGroup=&RandomRangeInt(Minimum=0, Maximum=8)) {isSetParameters=true}
   /SetVariables(Particles.VisibilityTag=&SpawnGroup()) {isSetParameters=true, module=SetVariables_#2}
   &InitializeParticle(
    Color=<rgba 0.015 0.0104 0.0104 1>
    Lifetime=&RandomRangeFloat(Minimum=1.5, Maximum=1.7)
    Mesh Scale Mode=::NewEnumerator3
    UsePositionOffset=false
    Write Material Random=true
   )
   &CameraOffset(Camera Offset Amount=2)
   /SetVariables(Particles.Y Offset:NiagaraFloat=&RandomRangeFloat(Minimum=0, Maximum=1)) {isSetParameters=true, module=SetVariables_#3}
  ]
  particleUpdateScript [
   &ParticleState()
   &SkeletalMeshLocation(
    Bone / Socket Index=&SpawnGroup()
    Bone Sampling Mode=::NewEnumerator4
    MaskSpawnGroup=false
    Skeletal Mesh=$System.Skeletal Mesh
   ) {module=SkeletalMeshLocation001}
   &InterpolateOverTime(Rate Of Change=5, Target Vector Value:Vector3f=$Particles.SkeletalMeshLocation001.SampledPosition, Variable Type=::NewEnumerator2)
   /SetVariables(Particles.Y Offset:NiagaraFloat=&Add_Float(
     A=$Particles.Y Offset
     B=&Lerp_Float(
      A=0
      B=0.025
      Alpha=&FloatFromCurve(FloatCurve=DataInterface() {propertyValues {Curve {
          defaultValue=0
          keys [
           {interpMode=RCIM_Cubic, tangentMode=RCTM_User, time=0, value=0}
           {arriveTangent=1.7898, interpMode=RCIM_Cubic, leaveTangent=1.7898, tangentMode=RCTM_User, time=1, value=1}
          ]
         }}})
     )
    )) {isSetParameters=true, module=SetVariables_#4}
   &DynamicMaterialParameters(
    Index 0 Param 1=&FloatFromCurve(FloatCurve=DataInterface() {propertyValues {Curve {
        defaultValue=1
        keys [{interpMode=RCIM_Cubic, tangentMode=RCTM_User, time=0.7, value=1}, {interpMode=RCIM_Cubic, tangentMode=RCTM_User, time=1, value=0}]
       }}})
    Index 0 Param 2=&FloatFromCurve(FloatCurve=DataInterface() {propertyValues {Curve {
        defaultValue=1
        keys [{interpMode=RCIM_Cubic, tangentMode=RCTM_User, time=0.3, value=1}, {interpMode=RCIM_Cubic, tangentMode=RCTM_User, time=1, value=0}]
       }}})
    Index 0 Param 3=&Lerp_Float(
     A=1.3
     B=1.8
     Alpha=&FloatFromCurve(
      FloatCurve=DataInterface() {propertyValues {Curve {
         defaultValue=1
         keys [
          {interpMode=RCIM_Cubic, tangentMode=RCTM_User, time=0, value=0}
          {interpMode=RCIM_Cubic, tangentMode=RCTM_User, time=0.4, value=1}
          {tangentMode=RCTM_User, time=1, value=0}
         ]
        }}}
      CurveIndex=&ModuloFloat(
       A=&Add_Float(
        A=&Multiply_Float(A=$Emitter.Age, B=&RandomRangeFloat(Minimum=0.1, Maximum=0.25))
        B=&Multiply_FloatByInt(Float=0.17, Integer=$Particles.SpawnGroup)
       )
       B=1
      )
     )
    )
    Index 0 Param 4=&Lerp_Float(
     A=1
     B=1.5
     Alpha=&FloatFromCurve(FloatCurve=DataInterface() {propertyValues {Curve {
         defaultValue=0
         keys [
          {interpMode=RCIM_Cubic, tangentMode=RCTM_User, time=0, value=0}
          {arriveTangent=0.0049, interpMode=RCIM_Cubic, leaveTangent=0.0049, tangentMode=RCTM_User, time=1, value=1}
         ]
        }}})
    )
    Index 1 Param 1=&MakeFloatFromVector(Vector=$StackContext.InterpolateOverTime.InterpolateOverTime_V3_.Moving Average, Channel=::NewEnumerator0)
    Index 1 Param 2=&MakeFloatFromVector(Vector=$StackContext.InterpolateOverTime.InterpolateOverTime_V3_.Moving Average, Channel=::NewEnumerator1)
    Index 1 Param 3=&MakeFloatFromVector(Vector=$StackContext.InterpolateOverTime.InterpolateOverTime_V3_.Moving Average, Channel=::NewEnumerator2)
    Index 2 Param 1=$Particles.Y Offset
    Index 2 Param 2=&FloatFromCurve(FloatCurve=DataInterface() {propertyValues {Curve {
        defaultValue=1
        keys [
         {arriveTangent=-0.0614, interpMode=RCIM_Cubic, leaveTangent=-0.0614, tangentMode=RCTM_User, time=0, value=-0.05}
         {arriveTangent=1.0261, interpMode=RCIM_Cubic, leaveTangent=1.0261, tangentMode=RCTM_User, time=1, value=0.4}
        ]
       }}})
    Index 2 Param 3=&SelectFloatFromArray(
     Array Sampling Mode=::NewEnumerator0
     Direct Array Index=?
     Float Selection Array=DataInterface() {propertyValues {FloatData [1,0.5,0]}}
     Random Seed=0
    )
    Module.Index0Float0Write=true
    Module.Index0Float1Write=true
    Module.Index0Float2Write=true
    Module.Index0Float3Write=true
    Module.Index1Float0Write=true
    Module.Index1Float1Write=true
    Module.Index1Float2Write=true
    Module.Index1Float3Write=true
    Module.Index2Float0Write=true
    Module.Index2Float1Write=true
    Module.Index2Float2Write=true
    Module.Index2Float3Write=true
    Write Parameter Index 1=::NewEnumerator1
    Write Parameter Index 2=::NewEnumerator1
   ) {module=DynamicMaterialParameters001}
   &ScaleColor(
    Scale Alpha=&FloatFromCurve(FloatCurve=DataInterface() {propertyValues {Curve {
        defaultValue=1
        keys [
         {arriveTangent=5.1459, interpMode=RCIM_Cubic, leaveTangent=5.1459, tangentMode=RCTM_User, time=0, value=0}
         {interpMode=RCIM_Cubic, tangentMode=RCTM_User, time=0.5, value=1}
         {tangentMode=RCTM_User, time=1, value=0}
        ]
       }}})
    Scale RGB=&MakeVectorFromLinearColorRGB(Color=&FNTimeOfDayColor(Night Color=<rgba 0.4 0.4791 0.8 1>, Morning Color=<rgba 1 1 1 1>, Day Color=<rgba 1 1 1 1>, Evening Color=<rgba 1 1 1 1>))
    ScaleA=true
    ScaleRGB=true
   )
   &ScaleColor(
    Color Value To Scale=$Particles.Color
    Scale Alpha=&MakeCustomFloatFromBool(
     Bool=&Boolean_AndOperation(
      Bool1=$User.bisFootHidden
      Bool2=&Boolean_OrOperation(
       Bool1=&SetBoolByIntComparison(A=$Particles.SpawnGroup, B=4, Comparison Type=::NewEnumerator4)
       Bool2=&Boolean_OrOperation(
        Bool1=&SetBoolByIntComparison(A=$Particles.SpawnGroup, B=5, Comparison Type=::NewEnumerator4)
        Bool2=&Boolean_OrOperation(
         Bool1=&SetBoolByIntComparison(A=$Particles.SpawnGroup, B=6, Comparison Type=::NewEnumerator4)
         Bool2=&SetBoolByIntComparison(A=$Particles.SpawnGroup, B=7, Comparison Type=::NewEnumerator4)
        )
       )
      )
     )
     True Float=0
     False Float=1
    )
    ScaleA=true
    ScaleRGB=true
   ) {module=ScaleColor001}
   &ScaleColor(
    Color Value To Scale=$Particles.Color
    Scale Alpha=&MakeCustomFloatFromBool(
     Bool=&Boolean_AndOperation(
      Bool1=$User.bisHandHidden
      Bool2=&Boolean_OrOperation(
       Bool1=&SetBoolByIntComparison(A=$Particles.SpawnGroup, B=0, Comparison Type=::NewEnumerator4)
       Bool2=&SetBoolByIntComparison(A=$Particles.SpawnGroup, B=1, Comparison Type=::NewEnumerator4)
      )
     )
     True Float=0
     False Float=1
    )
    ScaleA=true
    ScaleRGB=true
   ) {module=ScaleColor002}
  ]
  rendererClasses [&NiagaraMeshRendererProperties]
  renderers [
   {rendererClass=&NiagaraMeshRendererProperties, rendererIndex=0}
   {rendererClass=&NiagaraMeshRendererProperties, rendererIndex=1}
   {rendererClass=&NiagaraMeshRendererProperties, rendererIndex=2}
   {rendererClass=&NiagaraMeshRendererProperties, rendererIndex=3}
   {rendererClass=&NiagaraMeshRendererProperties, rendererIndex=4}
   {rendererClass=&NiagaraMeshRendererProperties, rendererIndex=5}
   {rendererClass=&NiagaraMeshRendererProperties, rendererIndex=6}
   {rendererClass=&NiagaraMeshRendererProperties, rendererIndex=7}
  ]
 }
}
summary {
 systemName=NS_Character_SolidMist_Idle
 emitters [
  {emitterName=Attached_Smoke, rendererClasses [&NiagaraMeshRendererProperties]}
  {emitterName=Swiping_Smoke, rendererClasses [&NiagaraMeshRendererProperties]}
 ]
 userVariables [
  User.bisFootHidden {type=NiagaraBool}
  User.bIsFrontEnd {type=NiagaraBool}
  User.bIsFrontEndPreview {type=NiagaraBool}
  User.bisHandHidden {type=NiagaraBool}
 ]
}
