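# Animation Blueprint for the SolidMist medium male player body cosmetic, layering a Control Rig over the pose of the mesh it is attached to.
# Copy Pose From Mesh reads the attached parent's pose and feeds M_MED_SolidMist_CtrlRig, whose output reaches the final pose through a local-to-component-to-local round trip.
# Skydiving begin and end events toggle OnSkydiving, which drives the rig's isSkydive input.
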AnimBlueprint /BRCosmetics/Characters/Player/Male/Medium/Bodies/M_MED_SolidMist/Meshes/M_MED_SolidMist_AnimBP
target_skeleton=&/M_MED_SolidMist_Skeleton
graphs [
 AnimGraph {nodes [
   n1 = Control Rig!(Source=Copy Pose From Mesh() {bUseAttachedParent=true}, Alpha, isSkydive:bool=Get OnSkydiving() -> (OnSkydiving:bool)) -> (Pose) {
    alphaBoolBlend {blendOption=Linear}
    controlRigAssetReference {blueprintRigClass=&/M_MED_SolidMist_CtrlRig.M_MED_SolidMist_CtrlRig_C}
   }
   Output Pose(Result=Component To Local(ComponentPose=Local To Component(LocalPose=@n1))) {layerGroup=None}
  ]}
 EventGraph {nodes [
   Event OnEndSkydiving()
   > Set OnSkydiving(OnSkydiving:bool=false) -> (Output_Get:bool)
   Event OnBeginSkydiving()
   > Set OnSkydiving(OnSkydiving=true)
   Event BlueprintUpdateAnimation()
  ]}
]
variables [OnSkydiving {type=boolean}]
